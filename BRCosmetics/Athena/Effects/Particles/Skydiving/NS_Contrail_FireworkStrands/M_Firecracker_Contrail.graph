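# Default Lit material for Niagara mesh particles in the firework strands skydiving contrail, animating and coloring firecracker meshes.
# A switchable offset combines a time-driven sine wave, blended between two vectors by camera-relative position, with a Z-axis twist with normal fix.
# Distance from particle position shapes both, with Niagara dynamic parameters adjusting timing and falloff power.
# An optional texture mask tints red and green channels, with a UV gradient on red.
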Material /BRCosmetics/Athena/Effects/Particles/Skydiving/NS_Contrail_FireworkStrands/M_Firecracker_Contrail
graph {
 properties {
  bOutputTranslucentVelocity=true
  bUsedWithInstancedStaticMeshes=true
  bUsedWithNiagaraMeshParticles=true
  pixelDepthOffsetMode=PDOM_Legacy
  usedShadingModels=Default Lit
 }
 nodes [
  n1 = Time()
  n2 = DynamicParameter() {paramNames [Time offset, Param2, PowerMultiplier, Param4]}
  Add(A=@n1, B=@n2.Time offset)
  Constant() {r=0.3}
  n3 = Constant3Vector() {constant=<rgba 0 0 1 0>}
  n4 = Multiply(A=@n1, B=ScalarParameter() {defaultValue=-4, parameterName=TwistSpeed})
  n5 = WorldPosition()
  n6 = ParticlePositionWS()
  n7 = Saturate(None=Divide(A=Distance(A=@n5.XYZ, B=@n6)) {constB=111})
  n8 = ScalarParameter() {defaultValue=1.2465, parameterName=Power}
  n9 = DynamicParameter() {paramNames [Time offset, Param2, PowerMultiplier, Param4]}
  n10 = Add(
   A=Power(Base=OneMinus(None=@n7), Exp=Multiply(A=@n8, B=@n9.PowerMultiplier))
   B=ScalarParameter() {defaultValue=-1.0492, parameterName=Offset}
  )
  n11 = Multiply(A=Sine(None=Add(A=@n4, B=@n10)) {period=2}, B=ScalarParameter() {defaultValue=-0.2, parameterName=Twist})
  n12 = ObjectPositionWS()
  MaterialFunctionCall(Rotation Axis=@n3, Rotation Angle=@n11, Normals Pivot=@n12) {materialFunction=&FixRotateAboutAxisNormals}
  MaterialFunctionCall() {materialFunction=&BoundingBoxBased_0-1_UVW}
  Multiply(A=ComponentMask(None=ObjectPositionWS()) {r=true}, B=ScalarParameter() {parameterName=PositionInfluence})
  OneMinus(None=ComponentMask(None=TextureCoordinate()) {g=true})
  OneMinus()
  ParticlePositionWS()
  Power(
   Base=TextureSample() {texture=&2ndInfectedMapLinear}.R
   Exp=ScalarParameter() {defaultValue=0.5, parameterName=RoughnessPower}
  )
  n13 = WorldPosition()
  n14 = Multiply(
   A=ComponentMask(None=TransformPosition(None=@n13.XYZ) {transformSourceType=TRANSFORMPOSSOURCE_World}) {b=true}
   B=ScalarParameter() {defaultValue=0.01, parameterName=Freq}
  )
  n15 = Multiply(A=Multiply(A=@n1, B=@n2.Time offset), B=ScalarParameter() {defaultValue=-1.5, parameterName=Speed})
  n16 = VectorParameter() {parameterName=HeightOffset_Top, defaultValue=<rgba 15 0 0 0>}
  n17 = VectorParameter() {parameterName=HeightOffset_Front, defaultValue=<rgba 15 0 0 0>}
  n18 = Normalize(VectorInput=TransformPosition(None=ObjectPositionWS()) {transformType=TRANSFORMPOSSOURCE_Camera})
  n19 = Abs(None=ComponentMask(None=@n18) {r=true})
  n20 = Abs(None=ComponentMask(None=@n18) {g=true})
  n21 = Multiply(
   A=Sine(None=Add(A=@n14, B=@n15))
   B=LinearInterpolate(
    A=Transform(None=@n16.RGB) {transformSourceType=TRANSFORMSOURCE_Local}
    B=Transform(None=@n17.RGB) {transformSourceType=TRANSFORMSOURCE_Local}
    Alpha=Add(A=@n19, B=@n20)
   )
  )
  n22 = RotateAboutAxis(NormalizedRotationAxis=@n3, RotationAngle=@n11, PivotPoint=@n12, Position=WorldPosition().XYZ)
  StaticSwitchParameter(True=Multiply(A=@n10, B=Add(A=@n21, B=@n22)), False=Constant()) {defaultValue=true, parameterName=USE_WPO}
  n23 = VectorParameter() {parameterName=COLOR_GreenChannel}
  n24 = TextureSample() {texture=&/T_FireCrackerTEX}
  n25 = VectorParameter() {parameterName=COLOR_RedChannel, defaultValue=<rgba 1 0 0.0313 0>}
  n26 = TextureCoordinate()
  n27 = LinearInterpolate(
   A=@n25.RGB
   B=VectorParameter() {parameterName=COLOR_RedChannel_GradientBottom, defaultValue=<rgba 0.8542 0.3426 0.4805 0>}.RGB
   Alpha=Power(Base=ComponentMask(None=@n26) {r=true}, Exp=ScalarParameter() {defaultValue=2, parameterName=REDCHANNEL_Power})
  )
  StaticSwitchParameter(True=Add(A=Multiply(A=@n23.RGB, B=@n24.G), B=Multiply(A=@n24.R, B=@n27)), False=@n25.RGB) {parameterName=UseMask}
 ]
}
