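# Niagara skydiving contrail cosmetic that trails firework strands and small exploding fireworks from the player's skeletal mesh.
# Kite-module meshes form spinning hand and foot strands, while Hand_Bullets launches flickering fire meshes from skeleton points through curl noise.
# Bullets emit location and death events; smoke, flame, flare, confetti and sparkle emitters have no active spawn modules, implying event-driven spawning.
# Front-end modifiers retune motion, lifetime and scale for menu display.
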NiagaraSystem /BRCosmetics/Athena/Effects/Particles/Skydiving/NS_Contrail_FireworkStrands/NS_Contrail_FireworkStrands
dependencies {
 usedDataInterfaces [&NiagaraDataInterfaceCurve, &NiagaraDataInterfaceColorCurve, &NiagaraDataInterfaceVectorCurve]
 usedDynamicInputs [
  &VectorFromCurve
  &TransformVector
  &Lerp_Vector
  &SetBoolByFloatComparison
 ]
 usedModules [
  &SystemState
  /Game/Effects/Niagara/Modules/Skydiving/FN_DivingAttributes/v5/FN_DivingAttributes
  &self:SystemSpawnScript.SystemScriptSource.SystemScriptGraph.NiagaraNodeAssignment_0.SetVariables_#1
  &self:SystemSpawnScript.SystemScriptSource.SystemScriptGraph.NiagaraNodeAssignment_1.SetVariables_#2
  &FN_Integrator
  &self:SystemSpawnScript.SystemScriptSource.SystemScriptGraph.NiagaraNodeAssignment_2.SetVariables_#3
 ]
}
emitters {
 Confetti {
  emitterUpdateScript [&EmitterState()]
  particleSpawnScript [
   /Niagara/Modules/Spawn/Initialization/InitializeParticle(
    Color=&RandomRangeLinearColor(Minimum=<rgba 0 0.538 1 1>, Maximum=<rgba 0.037 1 0.9618 1>)
    Lifetime=&RandomRangeFloat(Minimum=0.2, Maximum=0.6)
    Module.Write Color=true
    Module.Write Lifetime=true
    Module.Write Mass=true
    Module.Write Material Random=true
    Module.Write Position=false
    Module.Write Scale=false
    Module.Write SpriteRotation=true
    Module.Write SpriteSize=true
    Position=&SimulationPosition()
    Sprite Rotation=&RandomRangeFloat(Minimum=0, Maximum=360)
    Sprite Size=&Vector2DFromFloat(Value=&RandomRangeFloat(Minimum=2, Maximum=4))
   )
   &SphereLocation(
    Hemisphere Z=true
    MaskSpawnGroup=false
    Override Seed=false
    Sphere Radius=&RandomRangeFloat(Minimum=5, Maximum=10)
    Surface Only=false
    UseNonUniformScale=false
    UseOffset=false
    UseSphereOrientation=false
   )
   &AddVelocityFromPoint(Module.Use Falloff Distance=false, Velocity Strength=&RandomRangeFloat(Minimum=150, Maximum=300))
   &AddVelocity()
  ]
  particleUpdateScript [
   &ParticleState()
   &Drag(Drag=&RandomRangeFloat(Minimum=5, Maximum=10), Use Linear Drag=true, Use Rotational Drag=false)
   &InheritSourceMovement(
    Applied Force Scale=<-1.3 -1.3 -1>
    ApplyForce=true
    ApplyForceFalloff=false
    ApplyPosition=true
    ApplyPositionFalloff=false
    Module.LimitForceSpeed=false
   )
   &GravityForce(Gravity=&RandomRangeVector(Minimum=<0 0 -360>, Maximum=<1 1 -200>))
   &SolveForcesAndVelocity(Clamp Velocity=false, Limit Acceleration=false)
   &ScaleSpriteSize(Scale Factor=&Vector2DFromFloat(Value=&FloatFromCurve(FloatCurve=DataInterface() {propertyValues {Curve {keys [
          {arriveTangent=-0.4596, interpMode=RCIM_Cubic, leaveTangent=-0.4596, tangentMode=RCTM_User, time=0.55, value=1}
          {arriveTangent=-4.9031, interpMode=RCIM_Cubic, leaveTangent=-4.9031, tangentMode=RCTM_User, time=1, value=0}
         ]}}})))
   &ScaleColor(
    Scale RGB=&MakeVector(
     X=1
     Y=&Add_Float(
      A=&RandomRangeFloat(Minimum=0, Maximum=4000)
      B=&Lerp_Float(A=0, B=&RandomRangeFloat(Minimum=25, Maximum=150), Alpha=$Particles.NormalizedAge)
     )
     Z=1
    )
    ScaleA=false
    ScaleRGB=true
   ) {enabled=false}
   /Niagara/Modules/Update/Color/ScaleColorBySpeed() {enabled=false}
   &Color() {enabled=false}
  ]
  rendererClasses [&NiagaraSpriteRendererProperties]
  renderers [{rendererClass=&NiagaraSpriteRendererProperties}]
 }
 FN_Contrails_Kite_v2 {
  emitterUpdateScript [
   &EmitterScalability(
    Module.bApplySpawnCountScale=false
    Module.bApplySpawnCountScaleByDistanceFraction=false
    Module.bUseMaxDistance=false
    Module.bUseMinDistance=false
   ) {enabled=false}
   &EmitterLifeCycle(NextLoopDuration=0.9833) {enabled=false}
   &EmitterState()
   /Game/Effects/Niagara/Modules/Spawn/FN_Spawn_Kite(
    "Link Per Kite  Amount"=1
    Noise=DataInterface() {propertyValues {
      XCurve {keys [
        <at 0 0 RCIM_Cubic>
        {arriveTangent=-22.177, interpMode=RCIM_Cubic, leaveTangent=-22.177, time=0.0148, value=0.3241}
        {arriveTangent=8.0832, interpMode=RCIM_Cubic, leaveTangent=8.0832, time=0.0653, value=-1.4471}
        {arriveTangent=39.4964, interpMode=RCIM_Cubic, leaveTangent=39.4964, time=0.1184, value=1.1614}
        <at 0.1293 1.083>
        {arriveTangent=-23.4377, interpMode=RCIM_Cubic, leaveTangent=-23.4377, time=0.1576, value=0.0233}
        {arriveTangent=-0.0218, interpMode=RCIM_Cubic, leaveTangent=-0.0218, time=0.2029, value=-0.642}
        <at 0.267 0.0209>
        {arriveTangent=-9.845, interpMode=RCIM_Cubic, leaveTangent=-9.845, time=0.2888, value=0.1309}
        {arriveTangent=3.938, interpMode=RCIM_Cubic, leaveTangent=3.938, time=0.3474, value=-0.7708}
        {arriveTangent=2.5316, interpMode=RCIM_Cubic, leaveTangent=2.5316, time=0.3869, value=0.5173}
        {arriveTangent=-4.044, interpMode=RCIM_Cubic, leaveTangent=-4.044, time=0.4237, value=-0.5776}
        <at 0.4469 0.2748>
        {arriveTangent=-4.0955, interpMode=RCIM_Cubic, leaveTangent=-4.0955, time=0.4551, value=0.4851}
        {arriveTangent=-8.4961, interpMode=RCIM_Cubic, leaveTangent=-8.4961, time=0.5107, value=0.0134}
        <at 0.5109 0.0104>
        {arriveTangent=-0.6728, interpMode=RCIM_Cubic, leaveTangent=-0.6728, time=0.5736, value=-1.125}
        <at 0.6118 -0.0575>
        {arriveTangent=17.1461, interpMode=RCIM_Cubic, leaveTangent=17.1461, time=0.6146, value=0.0105}
        {arriveTangent=11.3424, interpMode=RCIM_Cubic, leaveTangent=11.3424, time=0.6622, value=0.8072}
        {arriveTangent=-13.2445, interpMode=RCIM_Cubic, leaveTangent=-13.2445, time=0.699, value=0.9682}
        {arriveTangent=-25.0454, interpMode=RCIM_Cubic, leaveTangent=-25.0454, time=0.7226, value=0.0074}
        {arriveTangent=-0.028, interpMode=RCIM_Cubic, leaveTangent=-0.028, time=0.7672, value=-0.7386}
        {arriveTangent=21.9983, interpMode=RCIM_Cubic, leaveTangent=21.9983, time=0.7974, value=0.0053}
        {arriveTangent=5.888, interpMode=RCIM_Cubic, leaveTangent=5.888, time=0.8462, value=1.0004}
        {arriveTangent=-15.155, interpMode=RCIM_Cubic, leaveTangent=-15.155, time=0.8953, value=0.5818}
        <at 0.9119 0.0053>
        {arriveTangent=-7.8259, interpMode=RCIM_Cubic, leaveTangent=-7.8259, time=0.9321, value=-0.6098}
        <at 0.9591 -0.3645>
        {arriveTangent=8.9177, interpMode=RCIM_Cubic, leaveTangent=8.9177, time=0.9879, value=0}
        {arriveTangent=0.682, interpMode=RCIM_Cubic, leaveTangent=0.682, tangentMode=RCTM_User, time=1, value=0}
       ]}
      YCurve {keys [
        <at 0 0 RCIM_Cubic>
        {arriveTangent=0.0179, interpMode=RCIM_Cubic, leaveTangent=0.0179, time=0.0843, value=1.1292}
        {arriveTangent=-19.1001, interpMode=RCIM_Cubic, leaveTangent=-19.1001, time=0.117, value=0.0021}
        <at 0.1293 0.2701>
        {arriveTangent=-34.1213, interpMode=RCIM_Cubic, leaveTangent=-34.1213, time=0.1607, value=1.1225}
        {arriveTangent=-41.5782, interpMode=RCIM_Cubic, leaveTangent=-41.5782, time=0.1702, value=-1.125}
        {arriveTangent=2.3241, interpMode=RCIM_Cubic, leaveTangent=2.3241, time=0.2015, value=-0.5776}
        {arriveTangent=-2.7223, interpMode=RCIM_Cubic, leaveTangent=-2.7223, time=0.2533, value=-0.9318}
        <at 0.267 -0.7557>
        {arriveTangent=19.4781, interpMode=RCIM_Cubic, leaveTangent=19.4781, time=0.3065, value=0.1309}
        {arriveTangent=-8.6722, interpMode=RCIM_Cubic, leaveTangent=-8.6722, time=0.3637, value=1.1292}
        <at 0.4469 -1.0865>
        {arriveTangent=41.2205, interpMode=RCIM_Cubic, leaveTangent=41.2205, time=0.4687, value=-1.3504}
        {arriveTangent=49.9749, interpMode=RCIM_Cubic, leaveTangent=49.9749, time=0.4959, value=0.936}
        <at 0.5109 0.7611>
        {arriveTangent=3.1016, interpMode=RCIM_Cubic, leaveTangent=3.1016, time=0.5532, value=-0.5454}
        {arriveTangent=9.5865, interpMode=RCIM_Cubic, leaveTangent=9.5865, time=0.5777, value=0.9682}
        <at 0.6118 0.0165>
        {arriveTangent=11.842, interpMode=RCIM_Cubic, leaveTangent=11.842, time=0.6309, value=-0.4165}
        {arriveTangent=-2.9535, interpMode=RCIM_Cubic, leaveTangent=-2.9535, time=0.6731, value=0.7428}
        {arriveTangent=-5.1686, interpMode=RCIM_Cubic, leaveTangent=-5.1686, time=0.6854, value=-0.5776}
        {arriveTangent=7.3333, interpMode=RCIM_Cubic, leaveTangent=7.3333, time=0.7167, value=0.5173}
        {arriveTangent=-15.9313, interpMode=RCIM_Cubic, leaveTangent=-15.9313, time=0.7641, value=0}
        <at 0.8095 -0.9606>
        {interpMode=RCIM_Cubic, tangentMode=RCTM_User, time=0.8489, value=0}
        {interpMode=RCIM_Cubic, tangentMode=RCTM_User, time=0.904, value=0.6415}
        {interpMode=RCIM_Cubic, tangentMode=RCTM_User, time=0.9329, value=0.2589}
        {interpMode=RCIM_Cubic, tangentMode=RCTM_User, time=0.967, value=-0.5063}
        {interpMode=RCIM_Cubic, tangentMode=RCTM_User, time=1, value=0}
       ]}
      ZCurve {keys [
        {interpMode=RCIM_Cubic, tangentMode=RCTM_User, time=0, value=0}
        {arriveTangent=10.357, interpMode=RCIM_Cubic, leaveTangent=10.357, tangentMode=RCTM_User, time=0.0489, value=-1.0284}
        {arriveTangent=5.6189, interpMode=RCIM_Cubic, leaveTangent=5.6189, tangentMode=RCTM_User, time=0.0854, value=0.8726}
        {arriveTangent=-0.1676, interpMode=RCIM_Cubic, leaveTangent=-0.1676, tangentMode=RCTM_User, time=0.1245, value=-0.6036}
        {arriveTangent=-0.5858, interpMode=RCIM_Cubic, leaveTangent=-0.5858, tangentMode=RCTM_User, time=0.1956, value=0.8542}
        {interpMode=RCIM_Cubic, tangentMode=RCTM_User, time=0.219, value=-0.659}
        {interpMode=RCIM_Cubic, tangentMode=RCTM_User, time=0.2776, value=0.8542}
        {arriveTangent=-3.0163, interpMode=RCIM_Cubic, leaveTangent=-3.0163, tangentMode=RCTM_User, time=0.2916, value=-0.659}
        {arriveTangent=7.8759, interpMode=RCIM_Cubic, leaveTangent=7.8759, tangentMode=RCTM_User, time=0.3877, value=0.522}
        {arriveTangent=-10.0071, interpMode=RCIM_Cubic, leaveTangent=-10.0071, tangentMode=RCTM_User, time=0.4439, value=0.5405}
        {arriveTangent=-1.5965, interpMode=RCIM_Cubic, leaveTangent=-1.5965, tangentMode=RCTM_User, time=0.5205, value=-0.8066}
        {arriveTangent=1.6215, interpMode=RCIM_Cubic, leaveTangent=1.6215, tangentMode=RCTM_User, time=0.5595, value=0.3559}
        {arriveTangent=2.5933, interpMode=RCIM_Cubic, leaveTangent=2.5933, tangentMode=RCTM_User, time=0.6001, value=-0.6774}
        {arriveTangent=26.7367, interpMode=RCIM_Cubic, leaveTangent=26.7367, tangentMode=RCTM_User, time=0.6235, value=0.522}
        {arriveTangent=-16.0548, interpMode=RCIM_Cubic, leaveTangent=-16.0548, tangentMode=RCTM_User, time=0.6595, value=0.9095}
        {arriveTangent=-7.1505, interpMode=RCIM_Cubic, leaveTangent=-7.1505, tangentMode=RCTM_User, time=0.6845, value=-0.456}
        {arriveTangent=5.7226, interpMode=RCIM_Cubic, leaveTangent=5.7226, tangentMode=RCTM_User, time=0.7188, value=0.4851}
        {interpMode=RCIM_Cubic, tangentMode=RCTM_User, time=0.7641, value=0}
        <at 0.8148 -0.6258>
        {arriveTangent=6.4436, interpMode=RCIM_Cubic, leaveTangent=6.4436, time=0.8699, value=0.7849}
        {interpMode=RCIM_Cubic, tangentMode=RCTM_User, time=0.9119, value=0}
        <at 0.9565 -0.6736 RCIM_Cubic>
        <at 1 0>
       ]}
     }}
    Number of Bones=2
   )
  ]
  particleSpawnScript [
   /Game/Effects/Niagara/Modules/Kite/FN_Initialize_Kite_v2(Module.Write Color=true, Module.Write RibbonTwist=false, Module.Write RibbonWidth=true, Ribbon Width=0)
   /SetVariables(
    Particles.FeetScale:Vector3f=&MakeCustomVectorFromBool(Bool=HlslExpression() {hlslExpression="Particles.SpawnGroup < 2"}, True Vector=<0 0 0>, False Vector=<1 1 1>)
    Particles.HandsScale:Vector3f=&FN_FrontEndVector3Modifier(
     Gameplay=&MakeCustomVectorFromBool(Bool=HlslExpression() {hlslExpression="Particles.SpawnGroup < 2"}, True Vector=<3 2.5 2.25>, False Vector=<0 0 0>)
     Front End=&MakeCustomVectorFromBool(Bool=HlslExpression() {hlslExpression="Particles.SpawnGroup < 2"}, True Vector=<3.5 2.5 2>, False Vector=<0 0 0>)
    )
   ) {isSetParameters=true}
  ]
  particleUpdateScript [
   /Game/Effects/Niagara/Modules/Skydiving/FN_DivingAttributes/FN_DivingAttributes() {enabled=false}
   &KillParticles(Kill Particles=/Game/Effects/Niagara/DynamicInputs/State/FN_IsExecutionStateActive())
   /Game/Effects/Niagara/Modules/Kite/FN_Update_Kite_v2(
    Amplitude=0
    Frequency=0
    Kite Position Offset:Vector3f=<0 0 1>
    Kite Scale:Vector3f=&VectorFromFloat(Value=3)
    Rotation Amplitude=0
    Scale Kite Position=0.4545
    Scale String Start Pos=1
    Spool Position Offset:Vector3f=/BRCosmetics/Effects/Niagara/DynamicInputs/Skydiving/SkydiveHandsFeetVector3Modifier(Hands:Vector3f=<0 0 0>, Feet:Vector3f=<0 0 0>)
    Spool Scale:Vector3f=&VectorFromFloat(Value=/Game/Effects/Niagara/DynamicInputs/Skydiving/SkydiveHandsFeetFloatModifier(Hands:NiagaraFloat=1.5, Feet:NiagaraFloat=1.3))
    String End Offset:Vector3f=<0 0 0>
    String Length Limit=30
   )
   /Game/Effects/Niagara/Modules/Materials/FN_MeshMaterial(Amplitude=0, Falloff=500, Frequency=200, Offset:NiagaraFloat=$Particles.AccumulatedNormalizedDive, Scale Noise X Axis=0.21, Speed=0.75) {enabled=false}
   /BRCosmetics/Athena/Effects/Particles/Skydiving/NS_Contrail_AmmoBelt/Module/FN_QuaterionSwizzle(
    From Vector:Vector3f=&MakeVector(X=0, Y=HlslExpression() {hlslExpression="(Particles.UniqueID && 1) ? -1.0f : 1.0f"}, Z=0)
    Quat:Quat4f=$Particles.SampledRotation
    Set Mesh Orientation=true
    To Vector:Vector3f=<1 0 0>
   )
   /BRCosmetics/Athena/Effects/Particles/Skydiving/NS_Contrail_AmmoBelt/Module/FN_QuaternionSpin(
    Angle:NiagaraFloat=&Multiply_Float(
     A=$Emitter.Age
     B=&Multiply_Float(A=HlslExpression() {hlslExpression="(Particles.UniqueID && 1 ^ ((Particles.SpawnGroup < 2) ? 1 : 0)) ? -1.0f : 1.0f"}, B=1.5)
    )
    Quaternion:Quat4f=$Particles.FN_QuaterionSwizzle.Quat
    Set Mesh Orientation=true
   )
   /SetVariables(
    Particles.DynamicMaterialParameter:Vector4f=&MakeVector4(X=&MakeFloatFromInt(Int=$Particles.UniqueID), Y=0, Z=0, W=0)
    Particles.RenderPosition:Vector3f=&Add_Vector(A=$Particles.Spool.Position, B=<50 0 0>)
   ) {enabled=false, isSetParameters=true}
   &DynamicMaterialParameters(
    Index 0 Param 1=&MaskFloatBySpawnGroup(Masked Float=1.2, Passthrough Float=1)
    Index 0 Param 3=&FN_FrontEndFloatModifier(Gameplay=1, Front End=0.5)
    Module.Param0WriteEnabled=true
    Module.Param1WriteEnabled=true
    Module.Param2WriteEnabled=true
    Module.Param3WriteEnabled=true
   )
  ]
  rendererClasses [&NiagaraMeshRendererProperties]
  renderers [{rendererClass=&NiagaraMeshRendererProperties, rendererIndex=0}]
 }
 FlameLicks {
  emitterUpdateScript [
   &EmitterState()
   &SpawnBurst_Instantaneous(Spawn Count=&RandomRangeInt(Minimum=8, Maximum=12), Use Spawn Probability=false) {enabled=false}
  ]
  particleSpawnScript [
   &InitializeParticle(
    Color=<rgba 5 3.0056 1.6206 0.75>
    Lifetime=&UniformRangedFloat(Minimum=0.2, Maximum=0.3)
    Module.Write Color=false
    Module.Write Lifetime=true
    Module.Write Mass=false
    Module.Write Material Random=true
    Module.Write Position=false
    Module.Write Scale=false
    Module.Write SpriteRotation=true
    Module.Write SpriteSize=true
    Position=&SimulationPosition()
    Sprite Rotation=&UniformRangedFloat(Minimum=0, Maximum=360)
    Sprite Size=&Vector2DFromFloat(Value=&RandomRangeFloat(Minimum=20, Maximum=40))
   )
   &Color()
   &AccelerationForce(Acceleration=<0 0 50>)
   &SphereLocation(
    MaskSpawnGroup=false
    Override Seed=false
    Sphere Radius=10
    Surface Only=false
    UseNonUniformScale=false
    UseOffset=false
    UseSphereOrientation=false
   )
   &AddVelocityFromPoint(Module.Use Falloff Distance=false, Velocity Strength=&RandomRangeFloat(Minimum=20, Maximum=50))
   &DynamicMaterialParameters(
    Index 0 Param 2=&UniformRangedFloat(Minimum=0, Maximum=1)
    Index 0 Param 3=1
    Index 0 Param 4=&UniformRangedFloat(Minimum=1, Maximum=2)
    Module.Param0WriteEnabled=false
    Module.Param1WriteEnabled=true
    Module.Param2WriteEnabled=false
    Module.Param3WriteEnabled=true
   ) {module=DynamicMaterialParameters002}
   &CameraOffset(Camera Offset Amount=&RandomRangeFloat(Minimum=5, Maximum=10))
   &ApplyInitialForces(Apply Rotational Force to Rotational Velocity=false)
  ]
  particleUpdateScript [
   &ParticleState()
   &ScaleColor(
    Scale Mode=::NewEnumerator1
    Scale RGBA=&Vector4FromCurve(Vector4Curve=DataInterface() {propertyValues {
       AlphaCurve {keys [<at 0.0024 0>,<at 0.2857 1>,<at 0.75 1>,<at 1 0>]}
       BlueCurve {keys [<at 0 0>,<at 0.1771 0.2404>,<at 0.5441 0.0129>,<at 1 0>]}
       GreenCurve {keys [<at 0 0.6667>,<at 0.1771 1.7164>,<at 0.5441 0.1731>,<at 1 0.0333>]}
       RedCurve {keys [<at 0 1>,<at 0.1771 15>,<at 0.5441 1.2674>,<at 1 1>]}
      }})
    ScaleRGBA=true
   )
   &InheritSourceMovement(
    Applied Force Scale=<-1.3 -1.3 -1>
    ApplyForce=true
    ApplyForceFalloff=false
    ApplyPosition=true
    ApplyPositionFalloff=false
    Module.LimitForceSpeed=false
   )
   &ScaleSpriteSize(Scale Factor=&Vector2DFromFloat(Value=&FloatFromCurve(FloatCurve=DataInterface() {propertyValues {Curve {keys [
          {arriveTangent=7.5764, interpMode=RCIM_Cubic, leaveTangent=7.5765, tangentMode=RCTM_User, time=0, value=0.1}
          {arriveTangent=0.1045, interpMode=RCIM_Cubic, leaveTangent=0.1045, tangentMode=RCTM_User, time=0.25, value=1}
          {arriveTangent=-0.8944, interpMode=RCIM_Cubic, leaveTangent=-0.8944, tangentMode=RCTM_User, time=1, value=0.8}
         ]}}})))
   /Niagara/Modules/Update/SubUV/SubUVAnimation(Number Of Frames=64)
   &DynamicMaterialParameters(
    Index 0 Param 1=&FloatFromCurve(FloatCurve=DataInterface() {propertyValues {Curve {keys [{arriveTangent=2.7551, interpMode=RCIM_Cubic, leaveTangent=2.7551, tangentMode=RCTM_User, time=0, value=0}, <at 0.7 1>]}}})
    Index 0 Param 2=1
    Index 0 Param 3=1
    Index 0 Param 4=1
    Module.Param0WriteEnabled=true
    Module.Param1WriteEnabled=true
    Module.Param2WriteEnabled=true
    Module.Param3WriteEnabled=false
   ) {module=DynamicMaterialParameters001}
   &Drag(Drag=4, Use Linear Drag=true, Use Rotational Drag=false)
   &SolveForcesAndVelocity(Clamp Velocity=false, Limit Acceleration=false)
   /SetVariables(Particles.LightColor=&MultiplyLinearColor(
     Linear Color=&ColorFromCurve(ColorCurve=DataInterface() {propertyValues {
        AlphaCurve {keys [<at 0 1 RCIM_Cubic>,<at 1 0 RCIM_Cubic>]}
        BlueCurve {keys [<at 0 0 RCIM_Cubic>,<at 1 0 RCIM_Cubic>]}
        GreenCurve {keys [<at 0 6.7498 RCIM_Cubic>,<at 1 0 RCIM_Cubic>]}
        RedCurve {keys [<at 0 40 RCIM_Cubic>,<at 1 0 RCIM_Cubic>]}
       }})
     Scale Factor=&MakeVector4_FromFloat(Float=400)
    )) {isSetParameters=true}
  ]
  rendererClasses [&NiagaraSpriteRendererProperties, &NiagaraLightRendererProperties]
  renderers [
   {rendererClass=&NiagaraSpriteRendererProperties}
   {rendererClass=&NiagaraLightRendererProperties, rendererIndex=1}
  ]
 }
 Flare {
  emitterUpdateScript [&EmitterState()]
  particleSpawnScript [/Niagara/Modules/Spawn/Initialization/V2/InitializeParticle(
    Color Maximum=<rgba 1 0.1134 0 1>
    Color Minimum=<rgba 1 0.33 0 1>
    Color Mode=::NewEnumerator2
    Lifetime=0.15
    Sprite Rotation Mode=::NewEnumerator2
    Sprite Size Mode=::NewEnumerator1
    Uniform Sprite Size Max=200
    Uniform Sprite Size Min=150
    UsePositionOffset=false
    Write Material Random=true
   )]
  particleUpdateScript [
   &ParticleState()
   &InheritSourceMovement(
    Applied Force Scale=<-1.3 -1.3 -1>
    ApplyForce=true
    ApplyForceFalloff=false
    ApplyPosition=true
    ApplyPositionFalloff=false
    Module.LimitForceSpeed=false
   )
   &SolveForcesAndVelocity(Clamp Velocity=false, Limit Acceleration=false)
   &ScaleSpriteSize(Scale Factor=&Vector2DFromFloat(Value=&FloatFromCurve(FloatCurve=DataInterface() {propertyValues {Curve {keys [
          {interpMode=RCIM_Cubic, tangentMode=RCTM_User, time=0, value=0}
          {interpMode=RCIM_Cubic, tangentMode=RCTM_User, time=0.2, value=1}
          {tangentMode=RCTM_User, time=1, value=0}
         ]}}})))
   &ScaleColor(
    Scale Alpha=0.5
    Scale RGB=&VectorFromFloat(Value=&FloatFromCurve(
      FloatCurve=DataInterface() {propertyValues {Curve {keys [{interpMode=RCIM_Cubic, tangentMode=RCTM_User, time=0, value=1}, {interpMode=RCIM_Cubic, tangentMode=RCTM_User, time=1, value=0}]}}}
      Scale Curve=5
     ))
    ScaleA=true
    ScaleRGB=true
   )
  ]
  rendererClasses [&NiagaraSpriteRendererProperties, &NiagaraLightRendererProperties]
  renderers [
   {rendererClass=&NiagaraSpriteRendererProperties}
   {rendererClass=&NiagaraLightRendererProperties, rendererIndex=1}
  ]
 }
 Hand_Bullets {
  emitterSpawnScript [&FN_PreserveSkydiveUserVariables() {module=PreserveUserVariables}]
  emitterUpdateScript [
   &EmitterState()
   &SpawnRate(SpawnRate=10, Use Spawn Probability=false)
  ]
  particleSpawnScript [
   /Game/Effects/Niagara/Modules/Groups/FN_SetSpawnGroup_v1/FN_SetSpawnGroup(Emitter Bone Count=4) {enabled=false}
   &FN_SetSpawnGroup_v2(Number of Spawn Groups=2)
   /Niagara/Modules/Spawn/Initialization/InitializeParticle(
    Lifetime=&FN_FrontEndFloatModifier(
     Gameplay=&UniformRangedFloat(Minimum=0.3, Maximum=0.6)
     Front End=&UniformRangedFloat(Minimum=0.5, Maximum=0.9)
    )
    Mesh Scale=<1.5 1.5 1.5>
    Module.Write Color=true
    Module.Write Lifetime=true
    Module.Write Mass=true
    Module.Write Material Random=true
    Module.Write Position=false
    Module.Write Scale=true
    Module.Write SpriteRotation=false
    Module.Write SpriteSize=false
    Position=&SimulationPosition()
   )
   &SampleSkeletalMeshSkeleton(Mesh:NiagaraDataInterfaceSkeletalMesh=$User.SkeletalMesh, Specific Bone Index:NiagaraInt32=$Particles.SpawnGroup) {enabled=false}
   /Niagara/Modules/Spawn/Location/SkeletalMeshSkeletonLocation(Module.Inherit Velocity=false, Module.Use Sampled Bone as Ribbon ID=true) {enabled=false}
   &SkeletalMeshLocation(
    Bone / Socket Index=$Particles.SpawnGroup
    Bone Sampling Mode=::NewEnumerator4
    Inherit Velocity=false
    MaskSpawnGroup=false
    Sampled Position Offset=&FN_FrontEndVector3Modifier(Gameplay=<0 0 -20>, Front End=<0 0 -40>)
    Skeletal Mesh=$User.SkeletalMesh
   )
   /BRCosmetics/Athena/Effects/Particles/Skydiving/NS_Contrail_AmmoBelt/Module/FN_OrientMeshToEngineOwner(Set Mesh Orientation=true)
   /SetVariables(Particles.Position=&Add_Vector(
     A=Linked() {linkedVariable {name=Particles.Position, type=NiagaraPosition}}
     B=&Multiply_VectorByFloat(
      Vector=/Niagara/DynamicInputs/Helpers/SafeNormalizeVector(Vector To Normalize:Vector3f=$Particles.Velocity, Fallback Vector:Vector3f=<0 0 1>)
      Float=HlslExpression() {hlslExpression="Particles.SpawnGroup < 2 ?  95: 60"}
     )
    )) {isSetParameters=true}
   &AddVelocityInCone(
    Cone Angle=&FN_FrontEndFloatModifier(Gameplay=80, Front End=50)
    Cone Axis=<0 0 1>
    Cone Axis Coordinate Space=::World
    Module.Use Velocity Falloff On Cone Axis=false
    Velocity Strength=&FN_FrontEndFloatModifier(
     Gameplay=&UniformRangedFloat(Minimum=500, Maximum=1000)
     Front End=&RandomRangeFloat(Minimum=200, Maximum=400)
    )
   )
   &MeshRotationForce(
    "Lever Radius (cm)":NiagaraFloat=&UniformRangedFloat(Minimum=150, Maximum=350)
    Rotation=/Niagara/DynamicInputs/Random/RandomVector(
     Vector Scale:NiagaraFloat=1
     Constraint Vector:Vector3f=<0 0 1>
     Randomness Mode:ENiagaraRandomnessMode=::NewEnumerator0
     Random Seed:NiagaraInt32=0
     Fixed Random Seed:NiagaraBool=false
    )
   )
   &ApplyInitialForces()
  ]
  particleUpdateScript [
   &ParticleState()
   &GravityForce(Gravity=&FN_FrontEndVector3Modifier(Gameplay=<0 0 0>, Front End=<0 0 -900>))
   &ScaleMeshSize(Scale Factor=&VectorFromFloat(Value=&FloatFromCurve(FloatCurve=DataInterface() {propertyValues {Curve {keys [
          <at 0.0008 0 RCIM_Cubic>
          {interpMode=RCIM_Cubic, tangentMode=RCTM_User, time=0.1657, value=1.0109}
          {interpMode=RCIM_Cubic, tangentMode=RCTM_User, time=0.9, value=1.0073}
          {arriveTangent=-4.4535, interpMode=RCIM_Cubic, leaveTangent=-4.4535, tangentMode=RCTM_User, time=1, value=0}
         ]}}})))
   /SetVariables(Particles.FireColor:LinearColor=&ColorFromCurve(ColorCurve=DataInterface() {propertyValues {
       AlphaCurve {keys [<at 0 1 RCIM_Cubic>,<at 0.6965 1>,<at 1 0 RCIM_Cubic>]}
       BlueCurve {keys [<at 0 0 RCIM_Cubic>,<at 0.0682 0>,<at 0.139 0>,<at 0.2208 0>,<at 0.2835 0>,<at 0.4307 0>,<at 0.5371 0>,<at 0.6148 0>,<at 0.7184 0>,<at 0.8097 0>,<at 1 0 RCIM_Cubic>]}
       GreenCurve {keys [<at 0 0.0347 RCIM_Cubic>,<at 0.0682 1.184>,<at 0.139 0.0511>,<at 0.2208 0.4177>,<at 0.2835 0.5392>,<at 0.4307 0.7821>,<at 0.5371 0.0782>,<at 0.6148 0.1578>,<at 0.7184 0.5879>,<at 0.8097 0.3079>,<at 1 0.0212 RCIM_Cubic>]}
       RedCurve {keys [<at 0 2 RCIM_Cubic>,<at 0.0682 40>,<at 0.139 0.474>,<at 0.2208 1>,<at 0.2835 5>,<at 0.4307 5>,<at 0.5371 0.5>,<at 0.6148 5>,<at 0.7184 10>,<at 0.8097 10>,<at 1 0.1354 RCIM_Cubic>]}
      }})) {isSetParameters=true}
   /SetVariables(Particles.FIREQUAT:Quat4f=&MultiplyQuaternion(
     Quaternion A=$Output.SkeletalMeshLocation.SampledRotation
     Quaternion B=&MultiplyQuaternion(
      Quaternion A=&FindQuatBetweenTwoVectors(From Vector=<1 0 0>, To Vector=<0 0 0>)
      Quaternion B=&MakeQuatFromAxisAngle(Axis=<1 0 0>, AngleInDegrees=&RandomRangeFloat(Minimum=0, Maximum=360))
     )
    )) {isSetParameters=true}
   &InheritSourceMovement(
    Applied Force Scale=<-1.5 -1.5 -1>
    ApplyForce=true
    ApplyForceFalloff=false
    ApplyPosition=true
    ApplyPositionFalloff=false
    Module.LimitForceSpeed=false
   )
   /SetVariables(Particles.ScaleMeshFire:Vector3f=&VectorFromFloat(Value=&FloatFromCurve(
      FloatCurve=DataInterface() {propertyValues {Curve {keys [
          <at 0.0008 0 RCIM_Cubic>
          {interpMode=RCIM_Cubic, tangentMode=RCTM_User, time=0.1657, value=1.0109}
          {interpMode=RCIM_Cubic, tangentMode=RCTM_User, time=0.2, value=0.8}
          {interpMode=RCIM_Cubic, tangentMode=RCTM_User, time=0.3, value=1}
          {interpMode=RCIM_Cubic, tangentMode=RCTM_User, time=0.4, value=0.8}
          {interpMode=RCIM_Cubic, tangentMode=RCTM_User, time=0.5, value=1}
          {interpMode=RCIM_Cubic, tangentMode=RCTM_User, time=0.6, value=0.8}
          {interpMode=RCIM_Cubic, tangentMode=RCTM_User, time=0.7, value=1.3}
          {interpMode=RCIM_Cubic, tangentMode=RCTM_User, time=0.8, value=0.8}
          {interpMode=RCIM_Cubic, tangentMode=RCTM_User, time=0.8367, value=1.2}
          {interpMode=RCIM_Cubic, tangentMode=RCTM_User, time=0.8885, value=0.8}
          {interpMode=RCIM_Cubic, tangentMode=RCTM_User, time=0.9465, value=1.05}
          {arriveTangent=-4.4535, interpMode=RCIM_Cubic, leaveTangent=-4.4535, tangentMode=RCTM_User, time=1, value=0}
         ]}}}
      Scale Curve=0.08
     ))) {isSetParameters=true}
   &CurlNoiseForce(
    Curl Noise Cone Mask Axis:Vector3f=&Multiply_Vector(A=<1 1 1>, B=<1 1 1>)
    Mask Curl Noise=true
    Noise Strength=&FN_FrontEndFloatModifier(Gameplay=8000, Front End=0)
    Pan Noise=false
   )
   &Drag(
    Drag=&FN_FrontEndFloatModifier(Gameplay=15, Front End=0)
    Use Linear Drag=true
    Use Rotational Drag=false
   )
   &GravityForce() {enabled=false, module=GravityForce001}
   &SolveForcesAndVelocity(Clamp Velocity=false, Limit Acceleration=false)
   &SolveRotationalForcesAndVelocity()
   &GenerateDeathEvent(Event Probability=1, Use Event Probability=true)
   &GenerateLocationEvent(Use Event Probability=false, UseEventDelay=false)
   &InheritSourceMovement(ApplyForce=false, ApplyPosition=true, ApplyPositionFalloff=false) {enabled=false, module=InheritSourceMovement001}
  ]
  rendererClasses [&NiagaraMeshRendererProperties]
  renderers [
   {rendererClass=&NiagaraMeshRendererProperties, rendererIndex=0}
   {rendererClass=&NiagaraMeshRendererProperties, rendererIndex=1}
  ]
 }
 Smoke {
  emitterUpdateScript [
   &EmitterScalability(
    Module.bApplySpawnCountScale=false
    Module.bApplySpawnCountScaleByDistanceFraction=false
    Module.bUseMaxDistance=false
    Module.bUseMinDistance=false
   ) {enabled=false}
   &EmitterLifeCycle(NextLoopDuration=0) {enabled=false}
   &EmitterState(
    Enable Distance Culling=true
    Max Distance Response=::KillImmediately
    MaxDistance=7000
    Module.bUseMaxDistance=true
    Module.bUseMinDistance=false
    Scalability Mode=::NewEnumerator1
   )
   &SpawnBurst_Instantaneous(Spawn Count=&UniformRangedInt(Minimum=3, Maximum=5), Use Spawn Probability=false) {enabled=false}
  ]
  particleSpawnScript [
   &InitializeParticle(
    Color=&MakeLinearColorFromVectorAndFloat(
     "Vector (RGB)"=&MakeVectorFromLinearColorRGB(Color=<rgba 1 0.6432 0.6133 1>)
     "Float (Alpha)"=&UniformRangedFloat(Minimum=0.05, Maximum=0.1)
    )
    Lifetime=&UniformRangedFloat(Minimum=0.33, Maximum=0.44)
    Module.Write Color=true
    Module.Write Lifetime=true
    Module.Write Mass=true
    Module.Write Material Random=true
    Module.Write Position=false
    Module.Write Scale=false
    Module.Write SpriteRotation=true
    Module.Write SpriteSize=true
    Position=&SimulationPosition()
    Sprite Rotation=&NormalizedAngleToDegrees(NormalizedAngle=&UniformRangedFloat(Minimum=0, Maximum=1))
    Sprite Size=&Vector2DFromFloat(Value=&Multiply_Float(A=&UniformRangedFloat(Minimum=70, Maximum=120), B=0.7))
   )
   /SetVariables(Particles.InitSpriteRotRate=&UniformRangedFloat(Minimum=-3, Maximum=3)) {isSetParameters=true}
   &AddVelocityInCone(
    Cone Angle=0
    Module.Use Velocity Falloff On Cone Axis=false
    Velocity Strength=&UniformRangedFloat(Minimum=0, Maximum=0)
   )
   /SetVariables() {isSetParameters=true}
   &CameraOffset(Camera Offset Amount=&UniformRangedFloat(Minimum=-20, Maximum=20))
   &SphereLocation(MaskSpawnGroup=false, Override Seed=false, Surface Only=false, UseNonUniformScale=false, UseOffset=false, UseSphereOrientation=false) {enabled=false}
  ]
  particleUpdateScript [
   &ParticleState()
   &SpriteRotationRate(Rotation Rate=&Multiply_Float(
     A=$Particles.InitSpriteRotRate
     B=&FloatFromCurve(FloatCurve=DataInterface() {propertyValues {Curve {keys [<at -0.0046 1.0113 RCIM_Cubic>,<at 1.0011 0.09 RCIM_Cubic>]}}})
    ))
   &SubUVAnimation(Number Of Frames=64, SubUV Lookup Index=&Multiply_Float(A=$Particles.Age, B=0.66))
   &InheritSourceMovement(
    Applied Force Scale=<-1.3 -1.3 -1>
    ApplyForce=true
    ApplyForceFalloff=false
    ApplyPosition=true
    ApplyPositionFalloff=false
    Module.LimitForceSpeed=false
   )
   &SolveForcesAndVelocity(Clamp Velocity=false, Limit Acceleration=false)
   &Color(
    Color=$Particles.Initial.Color
    Scale Alpha=&FloatFromCurve(FloatCurve=DataInterface() {propertyValues {Curve {keys [<at 0 0.5 RCIM_Cubic>,<at 1.0011 0.002 RCIM_Cubic>]}}})
   )
   &ScaleSpriteSize(Scale Factor=&Vector2DFromFloat(Value=&FloatFromCurve(FloatCurve=DataInterface() {propertyValues {Curve {keys [
          {arriveTangent=2.9149, interpMode=RCIM_Cubic, leaveTangent=2.9149, tangentMode=RCTM_User, time=-0.003, value=0}
          {arriveTangent=0.9127, interpMode=RCIM_Cubic, leaveTangent=0.9127, tangentMode=RCTM_User, time=0.23, value=0.63}
          <at 1 1>
         ]}}})))
   /Game/Effects/Niagara/Modules/Materials/MaterialSmoke(
    Erode:NiagaraFloat=&FloatFromCurve(FloatCurve=DataInterface() {propertyValues {
       bOptimizeLUT=false
       bUseLUT=false
       Curve {keys [
         {arriveTangent=-0.0652, interpMode=RCIM_Cubic, leaveTangent=-0.0652, tangentMode=RCTM_User, time=0.11, value=0.97}
         {arriveTangent=-5.7797, interpMode=RCIM_Cubic, leaveTangent=-5.7797, tangentMode=RCTM_User, time=0.39, value=0.66}
         {arriveTangent=-0.342, interpMode=RCIM_Cubic, leaveTangent=-0.342, tangentMode=RCTM_User, time=0.9954, value=-0.74}
        ]}
      }})
    Shading:NiagaraFloat=&FloatFromCurve(FloatCurve=DataInterface() {propertyValues {
       bOptimizeLUT=false
       bUseLUT=false
       Curve {keys [
         <at 0.07 0.5 RCIM_Cubic>
         {arriveTangent=1.2205, interpMode=RCIM_Cubic, leaveTangent=1.2205, time=0.17, value=1.02}
         {interpMode=RCIM_Cubic, tangentMode=RCTM_User, time=0.3895, value=0.89}
         <at 0.8 0 RCIM_Cubic>
        ]}
      }})
   )
  ]
  rendererClasses [&NiagaraSpriteRendererProperties]
  renderers [{rendererClass=&NiagaraSpriteRendererProperties}]
 }
 Sparkles {
  emitterUpdateScript [
   &EmitterScalability(
    Module.bApplySpawnCountScale=false
    Module.bApplySpawnCountScaleByDistanceFraction=false
    Module.bUseMaxDistance=false
    Module.bUseMinDistance=false
   ) {enabled=false}
   &EmitterLifeCycle(NextLoopDuration=4.4208) {enabled=false}
   &EmitterState()
   &SpawnRate(
    SpawnRate=&FloatFromCurve(
     FloatCurve=DataInterface() {propertyValues {Curve {keys [<at -0.0046 1.0113>,<at 0.6325 0.0329>,<at 1.0011 0.002>]}}}
     CurveIndex=$Emitter.NormalizedLoopAge
     Scale Curve=24
    )
    Use Spawn Probability=false
   ) {enabled=false}
  ]
  particleSpawnScript [
   &InitializeParticle(
    Color=&ColorFromCurve(ColorCurve=DataInterface() {propertyValues {
       AlphaCurve {keys [<at -0.0023 1 RCIM_Cubic>,<at 0.9977 -0.003 RCIM_Cubic>]}
       BlueCurve {keys [<at -0.0023 0.2396 RCIM_Cubic>, {arriveTangent=-0.254, interpMode=RCIM_Cubic, leaveTangent=-0.254, time=0.5349, value=0.026}, <at 0.9408 0>]}
       GreenCurve {keys [
         <at -0.0023 0.9895 RCIM_Cubic>
         {arriveTangent=-1.0309, interpMode=RCIM_Cubic, leaveTangent=-1.0309, time=0.5349, value=0.4505}
         <at 0.9408 0.0172>
        ]}
       RedCurve {keys [<at -0.0023 2 RCIM_Cubic>, {arriveTangent=-1.4635, interpMode=RCIM_Cubic, leaveTangent=-1.4635, time=0.5349, value=1}, <at 0.9408 0.6198>]}
      }})
    Lifetime=&UniformRangedFloat(Minimum=0.2, Maximum=0.5)
    Module.Write Color=true
    Module.Write Lifetime=true
    Module.Write Mass=true
    Module.Write Material Random=true
    Module.Write Position=false
    Module.Write Scale=false
    Module.Write SpriteRotation=false
    Module.Write SpriteSize=true
    Position=&SimulationPosition()
    Sprite Size=<10 25>
   )
   &AddVelocity(Velocity=&MakeVector(X=0, Y=&UniformRangedFloat(Minimum=150, Maximum=250), Z=0)) {enabled=false}
   /Niagara/Modules/Spawn/Location/SphereLocation(MaskSpawnGroup=false, Module.Override Randomness=false, Non Uniform Scale=<1 1 0.1>, Sphere Radius=20, Surface Only=false)
   &AddVelocityFromPoint(
    Module.Use Falloff Distance=false
    Velocity Origin:Vector3f=&SimulationPosition()
    Velocity Strength=&UniformRangedFloat(Minimum=100, Maximum=200)
   )
   /Niagara/Modules/Update/Forces/CurlNoiseForce(Noise Frequency=1, Noise Strength=250, Pan Noise=true, Pan Noise Field=<0 0.5 0>)
   &ApplyInitialForces()
  ]
  particleUpdateScript [
   &ParticleState()
   &GravityForce(Gravity=<0 0 -900>) {enabled=false}
   &InheritSourceMovement(
    Applied Force Scale=<-1.3 -1.3 -1>
    ApplyForce=true
    ApplyForceFalloff=false
    ApplyPosition=true
    ApplyPositionFalloff=false
    Module.LimitForceSpeed=false
   )
   &CurlNoiseForce(Noise Frequency=0.02, Noise Strength=500, Pan Noise=false) {module=CurlNoiseForce001}
   &Drag(Drag=10, Use Linear Drag=true, Use Rotational Drag=false)
   &SolveForcesAndVelocity(Clamp Velocity=false, Limit Acceleration=false)
   &ScaleColor(
    Scale Alpha=&FloatFromCurve(FloatCurve=DataInterface() {propertyValues {Curve {keys [<at -0.0092 0.0144>,<at 0.0967 0.9767>,<at 1.0011 0.002>]}}})
    Scale RGB=<15 5 1>
    ScaleA=true
    ScaleRGB=true
   )
   &ScaleSpriteSize(Scale Factor=&Vector2DFromFloat(Value=&FloatFromCurve(FloatCurve=DataInterface() {propertyValues {Curve {keys [<at -0.0046 0>,<at 0.0873 0.9701>,<at 1.0011 0.002>]}}}, Scale Curve=0.1)))
   &ScaleSpriteSizeBySpeed(Max Scale Factor=<1 3>, Min Scale Factor=<1 1.5>, Velocity Threshold=250)
  ]
  rendererClasses [&NiagaraSpriteRendererProperties]
  renderers [{rendererClass=&NiagaraSpriteRendererProperties}]
 }
}
summary {
 systemName=NS_Contrail_FireworkStrands
 emitters [
  {emitterName=Hand_Bullets, rendererClasses [&NiagaraMeshRendererProperties]}
  {emitterName=FN_Contrails_Kite_v2, rendererClasses [&NiagaraMeshRendererProperties]}
  {emitterName=Smoke, rendererClasses [&NiagaraSpriteRendererProperties]}
  {emitterName=FlameLicks, rendererClasses [&NiagaraSpriteRendererProperties, &NiagaraLightRendererProperties]}
  {emitterName=Flare, rendererClasses [&NiagaraSpriteRendererProperties, &NiagaraLightRendererProperties]}
  {emitterName=Confetti, rendererClasses [&NiagaraSpriteRendererProperties]}
  {emitterName=Sparkles, rendererClasses [&NiagaraSpriteRendererProperties]}
 ]
 userVariables [
  User.bIsFrontEnd {type=NiagaraBool}
  User.bIsFrontEndPreview {type=NiagaraBool}
  User.ParaGlideLeanAlpha {type=NiagaraFloat, defaultValue=0, description=""}
  User.PlayerSpeed {type=NiagaraFloat, defaultValue=0, description=""}
  User.SkeletalMesh {
   type=NiagaraDataInterfaceSkeletalMesh
   defaultValue {
    struct=NiagaraExt_VariableValue_DataInterface
    value {dataInterface=&self:NiagaraDataInterfaceSkeletalMesh_0}
   }
  }
 ]
}
